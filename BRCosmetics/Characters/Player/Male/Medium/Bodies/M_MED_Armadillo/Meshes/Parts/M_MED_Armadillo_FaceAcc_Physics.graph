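# Physics asset for the face accessory of the male medium Armadillo player cosmetic, simulating its dangling braids.
# Five braid chains of three small capsule bodies each hang from the head, linked by tight swing and twist limits, stiffest on the right chain.
# Capsules on the head, neck and upper spine give the body's shape for collision, joined by constraints with free swing and twist.
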PhysicsAsset /BRCosmetics/Characters/Player/Male/Medium/Bodies/M_MED_Armadillo/Meshes/Parts/M_MED_Armadillo_FaceAcc_Physics
bodies {
 dyn_braid_0_lft [{shapeType=Capsule, length=1, radius=1, center=<1.8 0 0>, rotation=<rot 0 90 90>}]
 dyn_braid_0_mid [{shapeType=Capsule, length=1, radius=1, center=<1.8 0 0>, rotation=<rot 0 90 90>}]
 dyn_braid_0_mid_lft [{shapeType=Capsule, length=1, radius=1, center=<1.8 0 0>, rotation=<rot 0 90 90>}]
 dyn_braid_0_mid_rgt [{shapeType=Capsule, length=1, radius=1, center=<1.8 0 0>, rotation=<rot 0 90 90>}]
 dyn_braid_0_rgt [{shapeType=Capsule, length=1, radius=1, center=<1.8 0 0>, rotation=<rot 0 90 90>}]
 dyn_braid_1_lft [{shapeType=Capsule, length=1.25, radius=0.75, center=<1.8 0 0>, rotation=<rot 0 90 90>}]
 dyn_braid_1_mid [{shapeType=Capsule, length=1.25, radius=0.75, center=<1.8 0 0>, rotation=<rot 0 90 90>}]
 dyn_braid_1_mid_lft [{shapeType=Capsule, length=1.25, radius=0.75, center=<1.8 0 0>, rotation=<rot 0 90 90>}]
 dyn_braid_1_mid_rgt [{shapeType=Capsule, length=1.25, radius=0.75, center=<1.8 0 0>, rotation=<rot 0 90 90>}]
 dyn_braid_1_rgt [{shapeType=Capsule, length=1.0408, radius=0.9592, center=<1.8 0 0>, rotation=<rot 0 90 90>}]
 dyn_braid_2_lft [{shapeType=Capsule, length=1, radius=1, center=<1.8 0 0>, rotation=<rot 0 90 90>}]
 dyn_braid_2_mid [{shapeType=Capsule, length=1, radius=1, center=<1.8 0 0>, rotation=<rot 0 90 90>}]
 dyn_braid_2_mid_lft [{shapeType=Capsule, length=1, radius=1, center=<1.8 0 0>, rotation=<rot 0 90 90>}]
 dyn_braid_2_mid_rgt [{shapeType=Capsule, length=1, radius=1, center=<1.8 0 0>, rotation=<rot 0 90 90>}]
 dyn_braid_2_rgt [{shapeType=Capsule, length=1, radius=1, center=<1.8 0 0>, rotation=<rot 0 90 90>}]
 head [{shapeType=Capsule, length=1.7388, radius=2.5207, shapeName=head_capsule, center=<7.5015 -0.2691 0.0499>, rotation=<rot 0 -36.9234 -86.2372>}]
 neck_01 [
  {shapeType=Capsule, length=2.25, radius=4.505, shapeName=spine_05_capsule, center=<2.2162 0 0>, rotation=<rot 90 0 0>}
  {shapeType=Capsule, length=24.5, radius=5.5, shapeName=neck_01_capsule, center=<17.0065 23.5155 0>}
 ]
 neck_02 [{shapeType=Capsule, length=1.25, radius=4.005, shapeName=spine_05_capsule, center=<2.4101 0 0>, rotation=<rot 90 0 0>}]
 spine_05 [
  {shapeType=Capsule, length=22.7502, radius=7.4999, shapeName=spine_05_capsule, center=<6.9564 4.6104 0.1183>}
  {shapeType=Capsule, length=7.3838, radius=6.75, shapeName=spine_05_capsule, center=<9.2628 0.2003 -0.333>}
 ]
}
constraints [
 {bone1Name=dyn_braid_0_mid_lft, bone2Name=head, swing1LimitDegrees=15, swing2LimitDegrees=15, twistLimitDegrees=2}
 {bone1Name=dyn_braid_1_mid_lft, bone2Name=dyn_braid_0_mid_lft, swing1LimitDegrees=15, swing2LimitDegrees=15, twistLimitDegrees=2}
 {bone1Name=dyn_braid_2_mid_lft, bone2Name=dyn_braid_1_mid_lft, swing1LimitDegrees=15, swing2LimitDegrees=15, twistLimitDegrees=2}
 {bone1Name=dyn_braid_0_mid, bone2Name=head, swing1LimitDegrees=15, swing2LimitDegrees=15, twistLimitDegrees=2}
 {bone1Name=dyn_braid_1_mid, bone2Name=dyn_braid_0_mid, swing1LimitDegrees=15, swing2LimitDegrees=15, twistLimitDegrees=2}
 {bone1Name=dyn_braid_2_mid, bone2Name=dyn_braid_1_mid, swing1LimitDegrees=15, swing2LimitDegrees=15, twistLimitDegrees=2}
 {bone1Name=dyn_braid_0_mid_rgt, bone2Name=head, swing1LimitDegrees=10, swing2LimitDegrees=10, twistLimitDegrees=2}
 {bone1Name=dyn_braid_1_mid_rgt, bone2Name=dyn_braid_0_mid_rgt, swing1LimitDegrees=15, swing2LimitDegrees=10, twistLimitDegrees=2}
 {bone1Name=dyn_braid_2_mid_rgt, bone2Name=dyn_braid_1_mid_rgt, swing1LimitDegrees=15, swing2LimitDegrees=15, twistLimitDegrees=2}
 {bone1Name=dyn_braid_0_lft, bone2Name=head, swing1LimitDegrees=15, swing2LimitDegrees=15, twistLimitDegrees=2}
 {bone1Name=dyn_braid_1_lft, bone2Name=dyn_braid_0_lft, swing1LimitDegrees=15, swing2LimitDegrees=15, twistLimitDegrees=2}
 {bone1Name=dyn_braid_2_lft, bone2Name=dyn_braid_1_lft, swing1LimitDegrees=15, swing2LimitDegrees=15, twistLimitDegrees=2}
 {bone1Name=dyn_braid_0_rgt, bone2Name=head, swing1LimitDegrees=8, swing2LimitDegrees=8, twistLimitDegrees=1}
 {bone1Name=dyn_braid_1_rgt, bone2Name=dyn_braid_0_rgt, swing1LimitDegrees=8, swing2LimitDegrees=8, twistLimitDegrees=1}
 {bone1Name=dyn_braid_2_rgt, bone2Name=dyn_braid_1_rgt, swing1LimitDegrees=8, swing2LimitDegrees=8, twistLimitDegrees=1}
 {
  bone1Name=neck_02
  bone2Name=neck_01
  swing1LimitDegrees=45
  swing1Motion=Free
  swing2LimitDegrees=45
  swing2Motion=Free
  twistLimitDegrees=45
  twistMotion=Free
 }
 {
  bone1Name=head
  bone2Name=neck_02
  swing1LimitDegrees=45
  swing1Motion=Free
  swing2LimitDegrees=45
  swing2Motion=Free
  twistLimitDegrees=45
  twistMotion=Free
 }
 {
  bone1Name=neck_01
  bone2Name=spine_05
  swing1LimitDegrees=45
  swing1Motion=Free
  swing2LimitDegrees=45
  swing2Motion=Free
  twistLimitDegrees=45
  twistMotion=Free
 }
]
